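# Physics asset for the GarlicWhisk backpack cosmetic, letting its small figure's head, torso, limbs and two ribbons swing within limits.
# Small base capsules for spine, head, right arm and ribbons lock angularly to body; base_arm_l has no such constraint.
# Capsule dyn_ chains hang from these bases with limited swing and twist, except twist-locked outer arm joints.
# Both leg bases attach to dyn_hips, so legs follow the spine-to-hips chain.
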PhysicsAsset /BRCosmetics/Accessories/FORT_Backpacks/Backpack_F_MED_GarlicWhisk/Meshes/F_MED_GarlicWhisk_Pack_Physics
bodies {
 base_arm_l [{shapeType=Capsule, length=0, radius=0.5, shapeName=base_arm_l_capsule, rotation=<rot 0 -90 -90>}]
 base_arm_r [{shapeType=Capsule, length=0, radius=0.5, shapeName=base_arm_r_capsule, rotation=<rot 0 180 -180>}]
 base_head [{shapeType=Capsule, length=0, radius=0.5, shapeName=base_head_capsule}]
 base_leg_l [{shapeType=Capsule, length=0, radius=0.5, shapeName=base_leg_l_capsule, rotation=<rot 180 0 -180>}]
 base_leg_r [{shapeType=Capsule, length=0, radius=0.5, shapeName=base_leg_r_capsule, rotation=<rot 0 180 -180>}]
 base_ribbon_A [{shapeType=Capsule, length=0, radius=0.505, shapeName=base_ribbon_A_capsule}]
 base_ribbon_B [{shapeType=Capsule, length=0, radius=0.505, shapeName=base_ribbon_B_capsule}]
 base_spine [{shapeType=Capsule, length=0, radius=0.5, shapeName=base_spine_capsule}]
 body [{shapeType=Capsule, length=0, radius=0.5, shapeName=body_capsule}]
 dyn_arm_l_01 [{
   shapeType=Capsule
   length=5.7284
   radius=1.5246
   shapeName=dyn_arm_l_01_capsule
   center=<-3.3513 0.2313 -0.3775>
   rotation=<rot 0 -89.9999 -95>
  }]
 dyn_arm_l_02 [{
   shapeType=Capsule
   length=1.4056
   radius=2.6335
   shapeName=dyn_arm_l_02_capsule
   center=<-0.9877 0.3734 0.2241>
   rotation=<rot -0.0001 -89.9998 -75>
  }]
 dyn_arm_r_01 [{
   shapeType=Capsule
   length=5.7284
   radius=1.5246
   shapeName=dyn_arm_r_01_capsule
   center=<3.3513 -0.2313 0.3775>
   rotation=<rot 0 -89.9999 85>
  }]
 dyn_arm_r_02 [{
   shapeType=Capsule
   length=1.4056
   radius=2.6335
   shapeName=dyn_arm_r_02_capsule
   center=<0.9877 -0.3734 -0.2241>
   rotation=<rot -0.0001 -89.9998 105>
  }]
 dyn_head [{shapeType=Capsule, length=0, radius=6.5, shapeName=dyn_head_capsule, center=<5.5955 0.7352 0>}]
 dyn_hips [{shapeType=Capsule, length=3.5, radius=5, shapeName=dyn_hips_capsule, center=<-2.1684 -1.4822 0.0734>}]
 dyn_leg_l_01 [{shapeType=Capsule, length=1.5, radius=3, shapeName=dyn_leg_l_01_capsule, center=<-1.7686 0.411 0>, rotation=<rot 90 0 0>}]
 dyn_leg_l_02 [{shapeType=Capsule, length=2, radius=3, shapeName=dyn_leg_l_02_capsule, center=<-1.2573 1.6383 -0.1522>, rotation=<rot 0 0.0001 90>}]
 dyn_leg_r_01 [{shapeType=Capsule, length=1.5, radius=3, shapeName=dyn_leg_r_01_capsule, center=<1.7686 -0.411 0>, rotation=<rot 90 0 180>}]
 dyn_leg_r_02 [{shapeType=Capsule, length=2, radius=3, shapeName=dyn_leg_r_02_capsule, center=<1.2573 -1.6383 0.1522>, rotation=<rot 0 0.0001 -90>}]
 dyn_ribbon_01 [{
   shapeType=Capsule
   length=2.7511
   radius=1
   shapeName=dyn_ribbon_01_capsule
   center=<0.5676 -0.4337 0.2631>
   rotation=<rot -0.0001 -89.9999 -89.9998>
  }]
 dyn_ribbon_02 [{shapeType=Capsule, length=2.0667, radius=1, shapeName=dyn_ribbon_02_capsule, center=<0.9006 -0.3319 -0.059>, rotation=<rot 0 -90 -84.9999>}]
 dyn_ribbon_03 [{shapeType=Capsule, length=1.005, radius=0.505, shapeName=dyn_ribbon_03_capsule, center=<0.3229 0.1005 0>, rotation=<rot 0 -90 -90>}]
 dyn_ribbon_04 [{shapeType=Capsule, length=2.451, radius=0.5, shapeName=dyn_ribbon_04_capsule, center=<0.5563 0 -0.1563>, rotation=<rot 0 -90 -90>}]
 dyn_spine [{shapeType=Capsule, length=3, radius=5.5, shapeName=dyn_spine_capsule, center=<0 -1.0744 0>}]
}
constraints [
 {bone1Name=dyn_head, bone2Name=base_head, swing1LimitDegrees=15, swing2LimitDegrees=12, twistLimitDegrees=10}
 {bone1Name=dyn_spine, bone2Name=base_spine, swing1LimitDegrees=10, swing2LimitDegrees=10, twistLimitDegrees=5}
 {bone1Name=dyn_hips, bone2Name=dyn_spine, swing1LimitDegrees=30, swing2LimitDegrees=15, twistLimitDegrees=5}
 {bone1Name=base_leg_l, bone2Name=dyn_hips, swing1LimitDegrees=30, swing2LimitDegrees=20, twistLimitDegrees=5}
 {bone1Name=dyn_leg_l_01, bone2Name=base_leg_l, swing1LimitDegrees=30, swing2LimitDegrees=20, twistLimitDegrees=5}
 {bone1Name=dyn_leg_l_02, bone2Name=dyn_leg_l_01, swing1LimitDegrees=15, swing2LimitDegrees=15, twistLimitDegrees=5}
 {
  bone1Name=base_spine
  bone2Name=body
  swing1LimitDegrees=45
  swing1Motion=Locked
  swing2LimitDegrees=45
  swing2Motion=Locked
  twistLimitDegrees=45
  twistMotion=Locked
 }
 {
  bone1Name=base_head
  bone2Name=body
  swing1LimitDegrees=10
  swing1Motion=Locked
  swing2LimitDegrees=10
  swing2Motion=Locked
  twistLimitDegrees=5
  twistMotion=Locked
 }
 {bone1Name=dyn_leg_r_01, bone2Name=base_leg_r, swing1LimitDegrees=30, swing2LimitDegrees=20, twistLimitDegrees=5}
 {bone1Name=dyn_leg_r_02, bone2Name=dyn_leg_r_01, swing1LimitDegrees=15, swing2LimitDegrees=15, twistLimitDegrees=5}
 {
  bone1Name=base_arm_r
  bone2Name=body
  swing1LimitDegrees=45
  swing1Motion=Locked
  swing2LimitDegrees=45
  swing2Motion=Locked
  twistLimitDegrees=45
  twistMotion=Locked
 }
 {bone1Name=base_leg_r, bone2Name=dyn_hips, swing1LimitDegrees=30, swing2LimitDegrees=20, twistLimitDegrees=5}
 {bone1Name=dyn_arm_l_02, bone2Name=dyn_arm_l_01, swing1LimitDegrees=15, swing2LimitDegrees=15, twistLimitDegrees=5, twistMotion=Locked}
 {bone1Name=dyn_arm_l_01, bone2Name=base_arm_l, swing1LimitDegrees=40, swing2LimitDegrees=45, twistLimitDegrees=10}
 {bone1Name=dyn_arm_r_01, bone2Name=base_arm_r, swing1LimitDegrees=45, swing2LimitDegrees=45, twistLimitDegrees=10}
 {bone1Name=dyn_arm_r_02, bone2Name=dyn_arm_r_01, swing1LimitDegrees=20, swing2LimitDegrees=20, twistLimitDegrees=5, twistMotion=Locked}
 {
  bone1Name=base_ribbon_A
  bone2Name=body
  swing1LimitDegrees=45
  swing1Motion=Locked
  swing2LimitDegrees=45
  swing2Motion=Locked
  twistLimitDegrees=45
  twistMotion=Locked
 }
 {bone1Name=dyn_ribbon_01, bone2Name=base_ribbon_A, swing1LimitDegrees=30, swing2LimitDegrees=45, twistLimitDegrees=5}
 {bone1Name=dyn_ribbon_02, bone2Name=dyn_ribbon_01, swing1LimitDegrees=30, swing2LimitDegrees=20, twistLimitDegrees=5}
 {
  bone1Name=base_ribbon_B
  bone2Name=body
  swing1LimitDegrees=45
  swing1Motion=Locked
  swing2LimitDegrees=45
  swing2Motion=Locked
  twistLimitDegrees=45
  twistMotion=Locked
 }
 {bone1Name=dyn_ribbon_03, bone2Name=base_ribbon_B, swing1LimitDegrees=20, swing2LimitDegrees=20, twistLimitDegrees=5}
 {bone1Name=dyn_ribbon_04, bone2Name=dyn_ribbon_03, swing1LimitDegrees=30, swing2LimitDegrees=15, twistLimitDegrees=10}
]
